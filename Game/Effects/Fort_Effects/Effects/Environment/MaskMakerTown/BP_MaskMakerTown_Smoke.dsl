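; MaskMaker Town smokestack effect actor that picks its smoke representation from scalability and platform.
; At BeginPlay, Epic effects quality with r.HeterogeneousVolumes at 1 shows the volume (HV), hides SplineSmoke and deactivates Niagara.
; Otherwise it hides HV, enabling the smoke mesh and Niagara stack for disabled volumes, lower quality, Switch or mobile.
; A bound scalability-change event repeats this logic, but its first check reuses values from BeginPlay.
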(event EventBeginPlay
  (bind _ns_nigarasmokestack (Variables|Default|GetNS_NigaraSmokeStack))
  ; 3 = Epic, 2= High, 1= Medium, 0=Low
  (bind _returnvalue (Settings|GetGameUserSettings))
  (Variables|Default|SetCurrent%20Scalability (Settings|GetVisualEffectQuality _returnvalue))
  (bind _returnvalue_2 (Development|GetConsoleVariableIntValue "r.HeterogeneousVolumes"))
  (Default|BindScalabilitySettings (delegate Custom|On%20Scalability%20Changed))
  (if (and (== (Settings|GetVisualEffectQuality _returnvalue) 3) (== _returnvalue_2 1))
    (Rendering|SetVisibility (Variables|Default|GetHV) true false)
    (Rendering|SetVisibility (Variables|Default|GetSplineSmoke) false false)
    (Components|Activation|Deactivate (Variables|Default|GetNS_NigaraSmokeStack))
    (else
      (Utilities|FlowControl|Sequence
        (:then_0
          (Rendering|SetVisibility (Variables|Default|GetHV) false false))
        (:then_1
          (bind _returnvalue_3 (Development|GetConsoleVariableIntValue "r.HeterogeneousVolumes"))
          ; 3 = Epic, 2= High, 1= Medium, 0=Low
          (bind _returnvalue_4 (Settings|GetGameUserSettings))
          (if (Math|Boolean|ORBoolean (== _returnvalue_3 0) (or (TouchPlatform|IsNintendoSwitchorNintendoSwitchPreview) (TouchPlatform|IsMobileGame)) (< (Settings|GetVisualEffectQuality _returnvalue_4) 3))
            (Rendering|SetVisibility (Variables|Default|GetSM_MaskMakerLair_Smoke) true true)
            (Components|Activation|Activate _ns_nigarasmokestack false)
            (Rendering|SetVisibility _ns_nigarasmokestack true false)))))))

(event Collision|EventActorBeginOverlap
  (params (OtherActor :type Actor%20Object%20Reference)))

(event EventTick
  (params (DeltaSeconds :type Float%20%28single-precision%29)))

(event Custom|On%20Scalability%20Changed
  (bind _ns_nigarasmokestack (Variables|Default|GetNS_NigaraSmokeStack))
  (Development|PrintString "scalaiblity changed")
  (if (and (== (Settings|GetVisualEffectQuality (stale EventBeginPlay _returnvalue)) 3) (== (stale EventBeginPlay _returnvalue_2) 1))
    (Rendering|SetVisibility (Variables|Default|GetHV) true false)
    (Rendering|SetVisibility (Variables|Default|GetSplineSmoke) false false)
    (Components|Activation|Deactivate (Variables|Default|GetNS_NigaraSmokeStack))
    (else
      (Utilities|FlowControl|Sequence
        (:then_0
          (Rendering|SetVisibility (Variables|Default|GetHV) false false))
        (:then_1
          (bind _returnvalue (Development|GetConsoleVariableIntValue "r.HeterogeneousVolumes"))
          ; 3 = Epic, 2= High, 1= Medium, 0=Low
          (bind _returnvalue_2 (Settings|GetGameUserSettings))
          (if (Math|Boolean|ORBoolean (== _returnvalue 0) (or (TouchPlatform|IsNintendoSwitchorNintendoSwitchPreview) (TouchPlatform|IsMobileGame)) (< (Settings|GetVisualEffectQuality _returnvalue_2) 3))
            (Rendering|SetVisibility (Variables|Default|GetSM_MaskMakerLair_Smoke) true true)
            (Components|Activation|Activate _ns_nigarasmokestack false)
            (Rendering|SetVisibility _ns_nigarasmokestack true false)))))))

(fn ConstructionScript)
