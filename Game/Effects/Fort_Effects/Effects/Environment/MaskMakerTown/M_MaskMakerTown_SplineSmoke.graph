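# Translucent, two-sided lit material for MaskMakerTown smoke trails on spline meshes, with a purple emissive gradient along V.
# SplineThicken sets width, narrowed near the camera by a sphere mask, plus a bounds-weighted offset that bends strands between two vectors by vertex green.
# Vertex green also staggers panning of a smoke texture, eroded by subtracted cloud noise.
# The resulting mask is softened by depth, near-camera and camera-distance fades.
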Material /Game/Effects/Fort_Effects/Effects/Environment/MaskMakerTown/M_MaskMakerTown_SplineSmoke
graph {
 properties {bAutomaticallySetUsageInEditor=false, blendMode=BLEND_Translucent, twoSided=true, usedShadingModels=Default Lit}
 nodes [
  n1 = ScalarParameter() {defaultValue=250}
  n2 = OneMinus(None=SphereMask(A=WorldPosition().XYZ, B=CameraPositionWS(), Radius=ScalarParameter() {defaultValue=2500, parameterName=Param_8}) {hardnessPercent=0})
  n3 = ScalarParameter() {defaultValue=500, parameterName=Param_1}
  n4 = MaterialFunctionCall() {materialFunction=/Game/Effects/Niagara/MasterMaterials/Functions/InstanceBounds_0-1_UVW}
  n5 = ScalarParameter() {defaultValue=2, parameterName=Param_2}
  n6 = VectorParameter() {parameterName=Param_2, defaultValue=<rgba 250 0 -750 0>}
  n7 = VectorParameter() {defaultValue=<rgba 400 0 150 0>}
  n8 = VertexColor()
  Add(
   A=MaterialFunctionCall(WidthBase=Multiply(A=@n1, B=@n2), WidthTip=Multiply(A=@n3, B=@n2)) {materialFunction=&SplineThicken}.WorldPositionOffset
   B=Multiply(A=Power(Base=@n4.B, Exp=@n5), B=LinearInterpolate(A=@n6.RGB, B=@n7.RGB, Alpha=@n8.G))
  )
  n9 = ComponentMask(None=TextureCoordinate()) {g=true}
  n10 = ScalarParameter() {defaultValue=7, parameterName=Param_6}
  MaterialFunctionCall(Emissive=Multiply(A=Power(Base=@n9, Exp=@n10), B=VectorParameter() {parameterName=Emissive COlor, defaultValue=<rgba 1.4698 0 3 1>}.RGB)) {materialFunction=&MF_FX_EmissiveAdjustments}
  n11 = VertexColor()
  n12 = Multiply(A=Add(A=Time(), B=LinearInterpolate(Alpha=@n11.G) {constA=-10, constB=10}), B=ScalarParameter() {defaultValue=0.1, parameterName=Param_3})
  n13 = TextureSample(UVs=MaterialFunctionCall(XY=Panner(Time=@n12) {speedY=1}) {materialFunction=&Swizzle}.YX) {samplerType=SAMPLERTYPE_Grayscale, texture=/Game/Effects/Fort_Effects/Textures/Smoke/T_Smoke_Trail_Soft_NoClamp}
  n14 = ComponentMask(None=TextureCoordinate()) {g=true}
  n15 = Power(Base=ComponentMask(None=TextureCoordinate()) {g=true}, Exp=ScalarParameter() {defaultValue=6, parameterName=Param_5})
  n16 = VectorParameter() {parameterName=Param_3, defaultValue=<rgba 0.5 1 0 0>}
  n17 = Time()
  n18 = LinearInterpolate(Alpha=VertexColor().G) {constA=-10, constB=10}
  n19 = Panner(
   Coordinate=Multiply(A=TextureCoordinate(), B=ComponentMask(None=@n16.RGB) {g=true, r=true})
   Time=Multiply(A=Add(A=@n17, B=@n18), B=ScalarParameter() {defaultValue=0.1, parameterName=Param_3})
  ) {speedY=1}
  n20 = ComponentMask(None=TextureCoordinate()) {g=true}
  n21 = Subtract(
   A=Multiply(A=Multiply(A=@n13.G, B=@n14), B=OneMinus(None=@n15))
   B=LinearInterpolate(B=TextureSample(UVs=@n19) {texture=&AtmosphericCloudNoise01}.G, Alpha=OneMinus(None=@n20))
  )
  n22 = ScalarParameter() {defaultValue=750, parameterName=Depth Fade}
  n23 = ScalarParameter() {defaultValue=500, parameterName=Param_7}
  Multiply(
   A=Multiply(
    A=DepthFade(Opacity=@n21, FadeDistance=@n22)
    B=MaterialFunctionCall(Distance=@n23) {materialFunction=&Near_Camera_Fade}.Result
   )
   B=Power(Base=NamedRerouteUsage())
  )
  Multiply()
  NamedRerouteDeclaration(None=@n2) {name=Fade, nodeColor=<rgba 0.8516 1 0 1>}
  OneMinus()
  ScalarParameter() {parameterName=Param_4}
  VectorParameter() {parameterName=Param_1, defaultValue=<rgba 0.5136 0 1 1>}
 ]
}
