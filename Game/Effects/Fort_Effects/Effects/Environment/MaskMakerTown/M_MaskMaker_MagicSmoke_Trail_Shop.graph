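# Translucent, two-sided material for a magic smoke trail at the MaskMaker Town shop, with a purple tint scaled by time-of-day colors.
# Opacity sums two panning smoke samples, then applies soft end masks and near-camera, view-angle, inverted Fresnel and depth fades.
# Vertex offset math widens the strip along its tangent between base and end widths, adding a sine wobble along normals.
# Custom primitive data supplies those widths and both time offsets.
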Material /Game/Effects/Fort_Effects/Effects/Environment/MaskMakerTown/M_MaskMaker_MagicSmoke_Trail_Shop
graph {
 properties {bAutomaticallySetUsageInEditor=false, blendMode=BLEND_Translucent, twoSided=true, usedShadingModels=Default Lit}
 nodes [
  n1 = TextureCoordinate()
  n2 = Multiply(
   A=Subtract(A=ComponentMask(None=@n1) {r=true}) {constB=0.5}
   B=MaterialFunctionCall() {materialFunction=&TangentBasis}.X U
  )
  n3 = TextureCoordinate()
  n4 = LinearInterpolate(
   A=ScalarParameter() {bUseCustomPrimitiveData=true, defaultValue=150, parameterName=Base}
   B=ScalarParameter() {bUseCustomPrimitiveData=true, defaultValue=300, parameterName=End, primitiveDataIndex=1}
   Alpha=OneMinus(None=ComponentMask(None=@n3) {g=true})
  )
  n5 = ScalarParameter() {defaultValue=0.45, parameterName=Param_3}
  n6 = Power(Base=OneMinus(None=VertexColor().R), Exp=ScalarParameter() {defaultValue=8, parameterName=Param_4})
  n7 = Multiply(A=VertexNormalWS(), B=ScalarParameter() {defaultValue=2, parameterName=Param_7})
  n8 = VectorParameter() {parameterName=Param_2, defaultValue=<rgba 0 50 50 0>}
  n9 = TextureCoordinate()
  n10 = Multiply(A=Time(), B=ScalarParameter() {defaultValue=0.01, parameterName=WPO Time Scale})
  n11 = Multiply(A=ComponentMask(None=Add(A=@n9, B=@n10)) {g=true}, B=ScalarParameter() {defaultValue=2, parameterName=Param_6})
  Add(A=Multiply(A=Multiply(A=@n2, B=@n4), B=LinearInterpolate(A=@n5, Alpha=@n6)), B=Multiply(A=Multiply(A=@n7, B=@n8.RGB), B=Sine(None=@n11)))
  n12 = Multiply(A=TextureCoordinate(), B=ComponentMask(None=VectorParameter() {defaultValue=<rgba 1 2 0 0>}.RGB) {g=true, r=true})
  n13 = Multiply(
   A=Add(A=Time(), B=ScalarParameter() {bUseCustomPrimitiveData=true, parameterName=TimeOffset 2, primitiveDataIndex=3})
   B=ScalarParameter() {defaultValue=0.05, parameterName=Param_2}
  )
  n14 = TextureSample(UVs=MaterialFunctionCall(XY=Panner(Coordinate=@n12, Time=@n13) {speedY=1}) {materialFunction=&Swizzle}.YX) {samplerType=SAMPLERTYPE_Grayscale, texture=&T_Smoke_Trail_Soft}
  n15 = Multiply(
   A=TextureCoordinate()
   B=ComponentMask(None=VectorParameter() {parameterName=secondary tile, defaultValue=<rgba 1 2 0 0>}.RGB) {g=true, r=true}
  )
  n16 = Multiply(
   A=Add(A=Time(), B=ScalarParameter() {bUseCustomPrimitiveData=true, parameterName=TimeOffset 1, primitiveDataIndex=4})
   B=ScalarParameter() {defaultValue=0.05, parameterName=secondary spped}
  )
  n17 = TextureSample(UVs=MaterialFunctionCall(XY=Panner(Coordinate=@n15, Time=@n16) {speedY=1}) {materialFunction=&Swizzle}.YX) {samplerType=SAMPLERTYPE_Grayscale, texture=&T_Smoke_Trail_Soft}
  n18 = ComponentMask(None=TextureCoordinate()) {g=true}
  n19 = Constant()
  n20 = ScalarParameter() {defaultValue=10, parameterName=Radius}
  n21 = ScalarParameter() {defaultValue=0.9402, parameterName=Soften End}
  n22 = TextureCoordinate()
  n23 = Multiply(
   A=Multiply(A=Add(A=@n14.G, B=@n17.G), B=SphereMask(A=@n18, B=@n19, Radius=@n20, Hardness=@n21))
   B=SphereMask(
    A=ComponentMask(None=@n22) {g=true}
    B=Constant() {r=1}
    Radius=ScalarParameter() {defaultValue=10, parameterName=Radius}
    Hardness=ScalarParameter() {parameterName=Soften Start}
   )
  )
  n24 = MaterialFunctionCall(Distance=ScalarParameter() {defaultValue=400, parameterName=Near Cam Fade}) {materialFunction=&Near_Camera_Fade}
  n25 = ScalarParameter() {defaultValue=0.25}
  n26 = Multiply(
   A=Multiply(
    A=Multiply(A=@n23, B=@n24.Result)
    B=MaterialFunctionCall("FadeSensitivity (optional)"=@n25) {materialFunction=&FadeBasedOnViewAngle}.Result
   )
   B=ScalarParameter() {defaultValue=1, parameterName=PreFresnelBOost}
  )
  n27 = OneMinus(None=Fresnel() {exponent=1})
  n28 = OneMinus(None=Fresnel() {exponent=2})
  n29 = ScalarParameter() {defaultValue=2, parameterName=fresnel}
  DepthFade(
   Opacity=Multiply(A=Multiply(A=@n26, B=@n27), B=Power(Base=@n28, Exp=@n29))
   FadeDistance=ScalarParameter() {defaultValue=150, parameterName=Param_1}
  )
  n30 = VectorParameter() {parameterName=Param_3, defaultValue=<rgba 0.4187 0.3099 0.2399 1>}
  n31 = VectorParameter() {parameterName=Param_4, defaultValue=<rgba 0.6187 0.4442 0.3511 1>}
  n32 = VectorParameter() {parameterName=Param_5, defaultValue=<rgba 0.3854 0.2836 0.2353 1>}
  n33 = VectorParameter() {parameterName=Param_6, defaultValue=<rgba 0.1782 0.0829 0.4396 1>}
  MaterialFunctionCall(Emissive=Multiply(
    A=VectorParameter() {parameterName=Param_1, defaultValue=<rgba 0.3861 0.15 0.75 0.5>}.RGB
    B=MaterialFunctionCall(Color_Morning=@n30.RGB, Color_Day=@n31.RGB, Color_Evening=@n32.RGB, Color_Night=@n33.RGB) {materialFunction=&Color_Time_of_Day}.Result
   )) {materialFunction=&MF_FX_EmissiveAdjustments}
  MaterialFunctionCall() {materialFunction=/Game/Effects/Fort_Effects/Materials/_Functions/CurveColor_Time_of_Day_S23}
  Multiply()
  n34 = VertexNormalWS()
  n35 = CameraVectorWS()
  Multiply(B=Saturate(None=DotProduct(A=@n34, B=@n35)))
 ]
}
