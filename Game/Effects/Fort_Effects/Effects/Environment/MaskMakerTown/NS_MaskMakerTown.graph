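# Niagara environment effect for MaskMakerTown that drifts colored, raymarched-named fog sprites along a user-supplied spline.
# One emitter spawns bursts and two rate-based groups, scaled by a probability parameter and tagged for visibility, with two sprite renderers.
# Spline position drives size, fade, camera offset, material parameters and a magenta-to-blue tint.
# Sprites also shrink near the player, dim slightly with time of day, and animate SubUV frames.
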NiagaraSystem /Game/Effects/Fort_Effects/Effects/Environment/MaskMakerTown/NS_MaskMakerTown
dependencies {
 usedDataInterfaces [
  &NiagaraDataInterfaceCurve
  &NiagaraDataInterfaceColorCurve
  &NiagaraDataInterfaceSpriteRendererInfo
 ]
 usedDynamicInputs [&Subtract_Float]
 usedModules [
  &self:GetSplineLengthSystem
  &SystemState
  &self:SystemSpawnScript.SystemScriptSource.SystemScriptGraph.NiagaraNodeAssignment_1.SetVariables_#1
  &self:SystemSpawnScript.SystemScriptSource.SystemScriptGraph.NiagaraNodeAssignment_0.SetVariables_#2
 ]
}
emitters {Fog-Raymarched {
  emitterUpdateScript [
   &EmitterState()
   &SpawnBurst_Instantaneous(Spawn Count=10, Spawn Probability=$User.Spawn Probability, Use Spawn Probability=true, UseLoopCountLimit=false)
   &SpawnRate(Spawn Probability=$User.Spawn Probability, SpawnRate=5, Use Spawn Probability=true) {module=SpawnRate001}
   &SpawnRate(Spawn Probability=$User.Spawn Probability, SpawnGroup=1, SpawnRate=2, Use Spawn Probability=true)
  ]
  particleSpawnScript [
   /SetVariables(Particles.VisibilityTag=&MakeCustomIntFromBool(
     Bool=&SetBoolByIntComparison(A=&SpawnGroup(), B=1, Comparison Type=::NewEnumerator4)
     True INT=1
     False INT=0
    )) {isSetParameters=true}
   /SetVariables(Particles.VisTag2:NiagaraInt32=&MakeCustomIntFromBool(
     Bool=&SetBoolByIntComparison(A=&SpawnGroup(), B=0, Comparison Type=::NewEnumerator4)
     True INT=1
     False INT=0
    )) {isSetParameters=true}
   &InitializeParticle(
    Color=<rgba 1 1 1 1.25>
    Lifetime Max=9
    Lifetime Min=4
    Lifetime Mode=::NewEnumerator1
    Sprite Rotation Angle=&RandomRangeFloat(Minimum=-0.25, Maximum=0.25)
    Sprite Rotation Mode=::NewEnumerator1
    Sprite Size Mode=::NewEnumerator1
    Uniform Sprite Size Max=1250
    Uniform Sprite Size Min=400
    UsePositionOffset=false
    Write Material Random=true
   )
   &self:SetSplineU(NumberofPos=20) {module=SetSplineU}
  ]
  particleUpdateScript [
   &ParticleState()
   &SpriteRotationRate(Delta Time:NiagaraFloat=$Particles.SplineU, Rotation Rate=&RandomRangeFloat(Minimum=0, Maximum=0.25))
   &ScaleSpriteSize(
    Uniform Curve Index=$Particles.SplineU
    Uniform Curve Sprite Scale=DataInterface() {propertyValues {Curve {keys [
        {interpMode=RCIM_Cubic, leaveTangent=7.5813, tangentMode=RCTM_Break, time=0.2, value=0}
        {arriveTangent=2.3459, interpMode=RCIM_Cubic, leaveTangent=2.5762, tangentMode=RCTM_Break, time=0.4, value=1.05}
        {interpMode=RCIM_Cubic, tangentMode=RCTM_Break, time=0.75, value=1.25}
       ]}}}
   ) {module=ScaleSpriteSize001}
   &ScaleSpriteSize(Uniform Curve Sprite Scale=DataInterface() {propertyValues {Curve {keys [
        {interpMode=RCIM_Cubic, leaveTangent=0.6142, tangentMode=RCTM_Break, time=0, value=0.75}
        {interpMode=RCIM_Cubic, leaveTangent=-0.0715, tangentMode=RCTM_Break, time=0.5, value=0.9875}
        <at 1 0.7375 RCIM_Cubic>
       ]}}})
   &ScaleColor(
    Curve Index=$Particles.SplineU
    Linear Color Curve=DataInterface() {propertyValues {
      AlphaCurve {keys [<at 0 0>,<at 0.4668 1>,<at 1 0>]}
      BlueCurve {keys [<at 0 1>,<at 1 1>]}
      GreenCurve {keys [<at 0 1>,<at 1 1>]}
      RedCurve {keys [<at 0 1>,<at 1 1>]}
     }}
    Scale Mode=::NewEnumerator2
   )
   &self:SampleAndMoveSplineU(
    Offset:Vector2f=<0 1>
    Spline:NiagaraDataInterfaceSpline=$User.Spline
    UPerSecond:NiagaraFloat=&RandomRangeFloat(Minimum=0.0175, Maximum=0.025)
   ) {module=SampleAndMoveSplineU}
   /SetVariables(Particles.Position=&AddVectorToPosition(Position=$Particles.Position, Vector=<0 0 0>)) {isSetParameters=true}
   &SubUVAnimation(
    Play Rate:NiagaraFloat=&RandomRangeFloat(Minimum=0.075, Maximum=0.1)
    "Sprite Renderer "=DataInterface() {propertyValues {SpriteRenderer=&self:Fog-Raymarched_2.NiagaraSpriteRendererProperties_0}}
    Start Frame Range Override=&RandomRangeInt(Minimum=0, Maximum=1)
    SubUV Animation Mode:ENiagara_SubUVLookupModeV2=::NewEnumerator3
    UseEndFrame=false
    UsePlayRateScale=false
    UseStartFrame=false
   )
   &DynamicMaterialParameters(
    Index 0 Param 1=&FloatFromCurve(
     FloatCurve=DataInterface() {propertyValues {Curve {keys [{interpMode=RCIM_Cubic, leaveTangent=6.4752, tangentMode=RCTM_Break, time=0, value=1.6}, <at 0.5 3>]}}}
     CurveIndex=$Particles.SplineU
    )
    Index 1 Param 1=0.8
    Index 1 Param 2=1
    Index 1 Param 3=&FloatFromCurve(FloatCurve=DataInterface() {propertyValues {Curve {keys [<at 0 0>,<at 0.1 0.2>,<at 0.4 0.6>]}}}, CurveIndex=$Particles.SplineU)
    Module.Param0WriteEnabled=true
    Module.Param0WriteEnabled1=true
    Module.Param1WriteEnabled=true
    Module.Param1WriteEnabled1=true
    Module.Param2WriteEnabled=true
    Module.Param2WriteEnabled1=true
    Module.Param3WriteEnabled=true
    Module.Param3WriteEnabled1=true
    Write Parameter Index 1=true
   ) {module=DynamicMaterialParameters001}
   &CameraOffset(Camera Offset Amount=&FloatFromCurve(FloatCurve=DataInterface() {propertyValues {Curve {keys [<at 0 1>,<at 1 0>]}}}, CurveIndex=$Particles.SplineU, Scale Curve=0)) {module=CameraOffset001}
   &ScaleColor(
    Scale Alpha=&Multiply_Float(
     A=&FloatFromCurve(
      FloatCurve=DataInterface() {propertyValues {Curve {keys [{interpMode=RCIM_Cubic, tangentMode=RCTM_User, time=0, value=0}, <at 0.5 1 RCIM_Cubic>, {tangentMode=RCTM_User, time=1, value=0}]}}}
      Scale Curve=&MaskFloatBySpawnGroup(Masked Float=1, Passthrough Float=1)
     )
     B=1
    )
    Scale RGB=&MakeVectorFromLinearColorRGB(Color=$Particles.Color)
    ScaleA=true
    ScaleRGB=true
   ) {module=ScaleColor001}
   /SetVariables(
    Particles.CamOff2:NiagaraFloat=&FloatFromCurve(FloatCurve=DataInterface() {propertyValues {Curve {keys [<at 0 1>,<at 1 0>]}}}, CurveIndex=$Particles.SplineU, Scale Curve=-50)
    Particles.Linear Color=&MakeLinearColorFromVectorAndFloat(
     "Vector (RGB)"=&MakeVectorFromLinearColorRGB(Color=$Particles.Color)
     "Float (Alpha)"=&FloatFromCurve(FloatCurve=DataInterface() {propertyValues {Curve {keys [<at 0.8 1>,<at 1 0>]}}})
    )
    Particles.Sort Order:NiagaraFloat=&Lerp_Float(A=0, B=1, Alpha=$Particles.SplineU)
    Particles.SpriteSize2=&Multiply_Vector2D_ByFloat(
     Float=&Multiply_Float(
      A=&MakeFloatFromVector2d(Vector 2D=$Particles.Initial.SpriteSize)
      B=&Multiply_Float(
       A=1.1
       B=&FloatFromCurve(
        FloatCurve=DataInterface() {propertyValues {Curve {keys [{arriveTangent=1.3931, interpMode=RCIM_Cubic, leaveTangent=7.2007, tangentMode=RCTM_Break, time=0.1, value=0}, <at 0.95 2>]}}}
        CurveIndex=$Particles.SplineU
       )
      )
     )
     Vector2D=<1 1>
    )
   ) {isSetParameters=true}
   &ScaleColor(
    Scale Mode=::NewEnumerator1
    Scale RGBA=&MakeVector4_FromLinearColor(LinearColor=&ColorFromCurve(
      ColorCurve=DataInterface() {propertyValues {
        AlphaCurve {keys [<at 0 1 RCIM_Cubic>,<at 0.1028 1>]}
        BlueCurve {keys [
          {arriveTangent=-0.2903, interpMode=RCIM_Cubic, time=0.0547, value=0.4438}
          <at 0.1667 0.1688>
          <at 0.25 0.0708>
          <at 0.5 0.0755>
          <at 0.75 0.1038>
          <at 0.875 0.0304>
          <at 0.9561 0.4438>
         ]}
        GreenCurve {keys [
          {arriveTangent=2.2098, interpMode=RCIM_Cubic, time=0.0547, value=0.0342}
          <at 0.1667 0.5721>
          <at 0.25 0.2534>
          <at 0.5 0.1109>
          <at 0.75 0.3716>
          <at 0.875 0.1925>
          <at 0.9561 0.0342>
         ]}
        RedCurve {keys [
          {arriveTangent=3.4655, interpMode=RCIM_Cubic, time=0.0547, value=0.3545}
          <at 0.1667 1>
          <at 0.25 0.4531>
          <at 0.5 0.5>
          <at 0.75 0.6646>
          <at 0.875 0.3646>
          <at 0.9561 0.3545>
         ]}
       }}
      CurveIndex=Linked() {linkedVariable {name=NPC.FortniteNPC.TimeOfDay_0-1, type=NiagaraFloat}}
     ))
    ScaleRGBA=true
   ) {enabled=false, module=ScaleColor002}
   &ScaleSpriteSize(
    Uniform Curve Index=&RemapRange(
     Input Value To Remap=&VectorLength(Vector=&Subtract_Position(A=&MakeVectorFromLinearColorRGB(Color=$NPC.FortniteNPC.PlayerPosition), B=$Particles.Position))
     Input Min=0
     Input Max=1250
     Output Min=0
     Output Max=1
    )
    Uniform Curve Sprite Scale=DataInterface() {propertyValues {Curve {keys [<at 0 0>, {arriveTangent=2.8897, interpMode=RCIM_Cubic, tangentMode=RCTM_Break, time=1, value=1}]}}}
   ) {module=ScaleSpriteSize002}
   &ScaleColor(
    Curve Index=$Particles.SplineU
    Linear Color Curve=DataInterface() {propertyValues {
      AlphaCurve {keys [<at 0 1>]}
      BlueCurve {keys [<at 0.0316 0.5374>,<at 0.4803 1>]}
      GreenCurve {keys [<at 0.0316 0>,<at 0.4803 0>]}
      RedCurve {keys [<at 0.0316 2>,<at 0.4803 0.2277>]}
     }}
    Scale Mode=::NewEnumerator2
   ) {module=ScaleColor003}
   &ScaleColor(
    Curve Index=Linked() {linkedVariable {name=NPC.FortniteNPC.TimeOfDay_0-1, type=NiagaraFloat}}
    Linear Color Curve=DataInterface() {propertyValues {
      AlphaCurve {keys [<at 0.4671 1>]}
      BlueCurve {keys [<at 0 0.5>,<at 0.25 0.5>,<at 0.4626 0.4>,<at 0.75 0.5>,<at 1 0.5>]}
      GreenCurve {keys [<at 0 0.5>,<at 0.25 0.5>,<at 0.4626 0.4>,<at 0.75 0.5>,<at 1 0.5>]}
      RedCurve {keys [<at 0 0.5>,<at 0.25 0.5>,<at 0.4626 0.4>,<at 0.75 0.5>,<at 1 0.5>]}
     }}
    Scale Mode=::NewEnumerator2
   ) {module=ScaleColor004}
  ]
  rendererClasses [&NiagaraSpriteRendererProperties]
  renderers [
   {rendererClass=&NiagaraSpriteRendererProperties}
   {rendererClass=&NiagaraSpriteRendererProperties, rendererIndex=1}
  ]
 }}
summary {
 systemName=NS_MaskMakerTown
 emitters [{emitterName=Fog-Raymarched, rendererClasses [&NiagaraSpriteRendererProperties]}]
 userVariables [
  User.End Position {type=NiagaraFloat, defaultValue=1, description=""}
  User.GlobalScaler {type=NiagaraFloat, defaultValue=1, description=""}
  User.Spawn Probability {type=NiagaraFloat, defaultValue=1, description=""}
  User.Spline {
   type=NiagaraDataInterfaceSpline
   description=""
   defaultValue {
    struct=NiagaraExt_VariableValue_DataInterface
    value {
     dataInterface=&self:NiagaraDataInterfaceSpline_0
     dataInterfaceClass=/Script/Niagara.NiagaraDataInterfaceSpline
    }
   }
  }
  User.Start Position {type=NiagaraFloat, defaultValue=0, description=""}
 ]
}
