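# Additive volume-domain material for heterogeneous volumes, rendering a sparse volume texture, likely an environment effect for MaskMakerTown.
# Panning world-space curl noise distorts the bounds-normalized local UVs that sample the volume's density and temperature, each optionally clamped.
# Temperature drives a switchable blackbody or tinted color, added to a density-weighted, density-offset sphere mask tint.
# Density is also multiplied by time-of-day colors from the Extinction Color parameter group.
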Material /Game/Effects/Fort_Effects/Effects/Environment/MaskMakerTown/SparseVolumeMaterial_NoEmissive
graph {
 properties {
  bUsedWithHeterogeneousVolumes=true
  blendMode=BLEND_Additive
  materialDomain=MD_Volume
  pixelDepthOffsetMode=PDOM_Legacy
  usedShadingModels=Default Lit
 }
 nodes [
  n1 = Multiply(
   A=VectorParameter() {group=Blackbody, parameterName=Blackbody Tint, defaultValue=<rgba 1 1 1 0>}.RGB
   B=ScalarParameter() {defaultValue=1, group=Blackbody, parameterName=Blackbody Scale}
  )
  n2 = ScalarParameter() {defaultValue=0.015, parameterName=Noise Distort}
  n3 = WorldPosition()
  n4 = Multiply(
   A=VectorParameter() {parameterName=Pan Speed, defaultValue=<rgba -250 0 -300 0>}.RGB
   B=Multiply(A=Time(), B=ScalarParameter() {defaultValue=0.1, parameterName=TimeScale})
  )
  n5 = TextureSample(UVs=Multiply(A=Add(A=@n3.XYZ, B=@n4), B=ScalarParameter() {defaultValue=0.0005, parameterName=Noise Scale})) {samplerType=SAMPLERTYPE_LinearColor, texture=/Game/Effects/FogVolume/VT_CurlNoise_LOW}
  n6 = ObjectLocalBounds()
  n7 = Subtract(
   A=MaterialFunctionCall() {materialFunction=&LocalPosition}.Local Position
   B=@n6.Min
  )
  n8 = SparseVolumeTextureSampleParameter!(UVs=Add(A=Multiply(A=@n2, B=@n5.RGB), B=Divide(A=@n7, B=@n6.Extents)), Tex) {parameterName=SparseVolumeTexture, sparseVolumeTexture=/Engine/EngineResources/EmptySparseVolumeTexture}
  n9 = StaticComponentMaskParameter(None=StaticSwitchParameter(True=@n8.Attributes B, False=@n8.Attributes B) {defaultValue=true, group=Temperature Mapping, parameterName="Temperature (Attributes B)"}) {defaultR=true, group=Temperature Mapping, parameterName=Temperature Mask}
  n10 = StaticSwitchParameter(
   True=Clamp(
    None=@n9
    Min=ScalarParameter() {group=Temperature, parameterName=Temperature Clamp Min}
    Max=ScalarParameter() {defaultValue=1, group=Temperature, parameterName=Temperature Clamp Max}
   )
   False=@n9
  ) {group=Temperature, parameterName=Clamp SVT Temperature}
  n11 = BlackBody!(Temp=Multiply(A=ScalarParameter() {defaultValue=4500, group=Blackbody, parameterName=Blackbody Temperature Scale}, B=@n10))
  n12 = Multiply(
   A=VectorParameter() {group=Temperature, parameterName=Temperature Color, defaultValue=<rgba 1 1 1 0>}.RGB
   B=ScalarParameter() {defaultValue=1, group=Temperature, parameterName=Temperature Scale}
  )
  n13 = VectorParameter() {parameterName=Param_1, defaultValue=<rgba 0.8716 0 1 1>}
  n14 = StaticComponentMaskParameter(None=StaticSwitchParameter(True=@n8.Attributes A, False=@n8.Attributes A) {defaultValue=true, group=Density Mapping, parameterName="Density (Attributes A)"}) {defaultR=true, group=Density Mapping, parameterName=Density Mask}
  n15 = StaticSwitchParameter(
   True=Clamp(
    None=@n14
    Min=ScalarParameter() {group=Density, parameterName=Density Clamp Min}
    Max=ScalarParameter() {defaultValue=1, group=Density, parameterName=Density Clamp Max}
   )
   False=@n14
  ) {group=Density, parameterName=Clamp SVT Density}
  n16 = ScalarParameter() {defaultValue=30}
  n17 = SphereMask(
   A=Add(A=LocalPosition() {localOrigin=Primitive}.XYZ, B=Multiply(A=@n15, B=@n16))
   B=VectorParameter() {defaultValue=<rgba -25 0 75 0>}.RGB
   Radius=ScalarParameter() {defaultValue=100, parameterName=Scale}
   Hardness=ScalarParameter() {parameterName=hardness}
  )
  Add(
   A=StaticSwitchParameter(True=Multiply(A=@n1, B=@n11), False=Multiply(A=@n10, B=@n12)) {defaultValue=true, group=Blackbody, parameterName=Use Blackbody Temperature}
   B=Multiply(A=Multiply(A=@n13.RGB, B=@n17), B=@n15)
  )
  Multiply()
  Multiply(
   A=VectorParameter() {group=Albedo, parameterName=Albedo Color, defaultValue=<rgba 1 1 1 0>}.RGB
   B=ScalarParameter() {defaultValue=1, group=Albedo, parameterName=Albedo Scale}
  )
  Multiply(B=MaterialFunctionCall(
    Value_Morning=ScalarParameter() {defaultValue=1, group=Density, parameterName=Density Scale Morning}
    Value_Day=ScalarParameter() {defaultValue=1, group=Density, parameterName=Density Scale Day}
    Value_Evening=ScalarParameter() {defaultValue=1, parameterName=Density Scale Evening}
    Value_Night=ScalarParameter() {defaultValue=1, parameterName=Density Scale Night}
   ) {materialFunction=&Scalar_Time_of_Day}.Result)
  n18 = VectorParameter() {group=Extinction Color, parameterName=TOD Morning, defaultValue=<rgba 5 2.7764 0.2396 1>}
  n19 = VectorParameter() {group=Extinction Color, parameterName=TOD Day, defaultValue=<rgba 5 2.7764 0.2396 1>}
  n20 = VectorParameter() {group=Extinction Color, parameterName=TOD Evening, defaultValue=<rgba 0 4.5124 5 1>}
  n21 = VectorParameter() {group=Extinction Color, parameterName=TOD Night, defaultValue=<rgba 5 3.6511 0 1>}
  Reroute(None=Multiply(
    A=@n15
    B=MaterialFunctionCall(Color_Morning=@n18.RGB, Color_Day=@n19.RGB, Color_Evening=@n20.RGB, Color_Night=@n21.RGB) {materialFunction=&Color_Time_of_Day}.Result
   ))
  Subtract(A=ActorPositionWS())
  WorldPosition()
 ]
}
